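# Material function for pet characters that sets the emissive color on incoming material attributes.
# An emissive texture, defaulting to black, is multiplied by a brightness parameter and a view-angle mask.
# The mask is the saturated dot product of pixel normal and camera vector raised to a power parameter, so glow is strongest facing the camera and fades toward edges.
# The result replaces any incoming emissive.
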MaterialFunction /Game/Characters/Pets/Materials/Functions/BaseFN_Pet_Emissives
graph {nodes [
  n1 = FunctionInput() {inputType=FunctionInput_MaterialAttributes}
  n2 = Multiply(
   A=ScalarParameter() {defaultValue=1, group=Emissive, parameterName=EmissiveBrightness}
   B=TextureSampleParameter2D() {group=Emissive, parameterName=Emissive, texture=&T_BlacksRGB}.RGB
  )
  n3 = DotProduct(A=PixelNormalWS(), B=CameraVectorWS())
  n4 = Saturate(None=Power(Base=Saturate(None=@n3), Exp=ScalarParameter() {defaultValue=1, group=Emissive, parameterName=EmissiveFresnelPower}))
  FunctionOutput(None=SetMaterialAttributes!(MaterialAttributes=GetMaterialAttributes(None=@n1).MaterialAttributes, EmissiveColor=Multiply(A=@n2, B=@n4)) {attributeSetTypes [#1]})
 ]}
