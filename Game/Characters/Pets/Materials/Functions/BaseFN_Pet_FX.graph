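# Pet material function that layers optional skin effects, fur fuzz and chameleon color shift, onto incoming material attributes.
# A shared SkinFX_Mask texture supplies each effect's alpha through its own static channel selection, both defaulting to red.
# Static switches chain the weapon-skin functions ClothFuzz, then MF_ChameleonColorShift, passing the material through when disabled.
# The AO input is declared but unconnected.
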MaterialFunction /Game/Characters/Pets/Materials/Functions/BaseFN_Pet_FX
graph {nodes [
  FunctionInput() {inputName=AO, inputType=FunctionInput_Scalar}
  n1 = TextureSampleParameter2D() {
   group=Skin FX
   parameterName=SkinFX_Mask
   samplerSource=SSM_Wrap_WorldGroupSettings
   samplerType=SAMPLERTYPE_Masks
   texture=&T_BlackMask
  }
  n2 = FunctionInput() {inputName=Material, inputType=FunctionInput_MaterialAttributes}
  n3 = StaticSwitchParameter(
   True=MaterialFunctionCall!(
    Alpha=StaticComponentMaskParameter(None=@n1.RGB) {defaultR=true, group=Skin FX, parameterName=Fur Channel}
    Gmap=Constant() {r=1}
    Material=@n2
   ) {materialFunction=/Game/Weapons/WeaponSkins/MaterialFunction/ClothFuzz}.Result
   False=@n2
  ) {group=Skin FX, parameterName=Use Fur}
  n4 = StaticComponentMaskParameter(None=@n1.RGB) {defaultR=true, group=Skin FX, parameterName=Chameleon Channel}
  FunctionOutput(None=StaticSwitchParameter(
    True=MaterialFunctionCall!(Material=@n3, Alpha=@n4) {materialFunction=/Game/Weapons/WeaponSkins/MaterialFunction/MF_ChameleonColorShift}.Result
    False=@n3
   ) {group=Skin FX, parameterName=Use Chameleon Colors})
 ]}
