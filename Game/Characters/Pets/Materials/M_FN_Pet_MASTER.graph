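# Master Default Lit material for pet characters, usable on skeletal meshes with morph targets and mesh deformers.
# It chains material functions via Material Attributes, starting with the shared character base, then pet emissives and pet FX.
# The base function's M output supplies its red channel as AO to the FX stage.
# The result then passes through the shared de-rez transition and character opacity functions.
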Material /Game/Characters/Pets/Materials/M_FN_Pet_MASTER
graph {
 properties {
  bUseMaterialAttributes=true
  bUsedWithMeshDeformer=true
  bUsedWithMorphTargets=true
  bUsedWithSkeletalMesh=true
  pixelDepthOffsetMode=PDOM_Legacy
  usedShadingModels=Default Lit
 }
 nodes [
  n1 = MaterialFunctionCall!(TexCoords) {materialFunction=/Game/Packages/Fortress_SharedMaterials/Base_Material_Functions/BaseCharacterShaderFunctions/BaseFN_CharacterFunction}
  n2 = MaterialFunctionCall!(In=@n1.Result) {materialFunction=/Game/Characters/Pets/Materials/Functions/BaseFN_Pet_Emissives}
  n3 = MaterialFunctionCall(Float3=@n1.M) {materialFunction=&BreakOutFloat3Components}
  MaterialFunctionCall!(Material=MaterialFunctionCall!(
    Fresnel
    Material=MaterialFunctionCall!(AO=@n3.R, Material=@n2.Result) {materialFunction=/Game/Characters/Pets/Materials/Functions/BaseFN_Pet_FX}.Result
   ) {materialFunction=/Game/Packages/Fortress_SharedMaterials/Base_Material_Functions/CharacterDeRez_Transition}.Result) {materialFunction=/Game/Packages/Fortress_SharedMaterials/Base_Material_Functions/BaseCharacterShaderFunctions/BaseFN_CharacterOpacity}
 ]
}
